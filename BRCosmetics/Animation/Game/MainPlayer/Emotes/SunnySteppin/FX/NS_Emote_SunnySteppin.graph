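# Niagara system for the Sunny Steppin emote: a one-shot burst of warm sun flash and scattering flowers.
# FlashElements layers spike, star and flare sprites; Sun_Light and two oscillating Sun_Rays emitters add the glow.
# FlowerParent throws two spawn groups of tumbling sprite/mesh flowers through curl noise and gravity, with collision gated by platform quality or front-end preview.
# All emitters briefly inherit horizontal source movement; front-end flags adjust colors and flash size.
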NiagaraSystem /BRCosmetics/Animation/Game/MainPlayer/Emotes/SunnySteppin/FX/NS_Emote_SunnySteppin
dependencies {
 usedDataInterfaces [
  &NiagaraDataInterfaceCurve
  &NiagaraDataInterfaceSpriteRendererInfo
  &NiagaraDataInterfaceArrayFloat
  &NiagaraDataInterfaceMeshRendererInfo
  &NiagaraDataInterfacePlatformSet
 ]
 usedDynamicInputs [&MakeCustomVectorFromBool, &RandomBool]
 usedModules [&SystemState]
}
emitters {
 FlashElements {
  emitterUpdateScript [
   &EmitterState()
   &SpawnBurst_Instantaneous(Spawn Group=1, Use Spawn Probability=false, UseLoopCountLimit=false) {module=SpawnBurst_Instantaneous001}
   &SpawnBurst_Instantaneous(
    Spawn Count=&RandomRangeInt(Minimum=6, Maximum=8)
    Use Spawn Probability=false
    UseLoopCountLimit=false
   )
  ]
  particleSpawnScript [
   &InitializeParticle(
    Color Mode=::NewEnumerator0
    Lifetime=&Lerp_Float(A=0.25, B=0.15, Alpha=$User.Variant)
    Sprite Rotation Mode=::NewEnumerator2
    UsePositionOffset=false
    Write Material Random=true
   )
   &ShapeLocation(MaskSpawnGroup=false, Override Seed=false, Sphere Radius=1, UseOwnerScale=true)
   &SpriteFacingAndAlignment(Sprite Alignment=$Particles.ShapeLocation.ShapeNormal, WriteAlignment=true, WriteFacing=false)
   &FN_PerParticleLOD(Enable SpriteSize Distance Scaling=true, Sprite Size Scale Far=<2.5 2.5>)
   /SetVariables(Particles.SpriteSize=<1 1>, Particles.VisibilityTag=&SpawnGroup()) {isSetParameters=true}
   /SetVariables(
    Particles.Flash Spike Color:LinearColor=&MakeCustomLinearColorFromBool(Bool=$User.bIsFrontEnd, True Color=<rgba 11 7.7647 0.2588 0.01>, False Color=<rgba 15 9.163 1.3596 0.01>)
    Particles.Flash Spike Lifetime:NiagaraFloat=&RandomRangeFloat(Minimum=0.05, Maximum=0.07)
    Particles.Flash Spike Sprite Size:Vector2f=&Multiply_Vector2D_ByFloat(
     Float=&FN_FrontEndFloatModifier(Gameplay=1, Front End=$User.FrontEndScale)
     Vector2D=&RandomRangeVector2D(Minimum=<1.5 27.5>, Maximum=<5 50>)
    )
   ) {isSetParameters=true}
   /SetVariables(
    Particles.Flash Star Cam Offset=15
    Particles.Flash Star Color:LinearColor=&MakeCustomLinearColorFromBool(Bool=$User.bIsFrontEnd, True Color=<rgba 11 7.7647 0.2588 0.01>, False Color=<rgba 15 9.163 1.3596 0.01>)
    Particles.Flash Star Lifetime=0.1
    Particles.Flash Star Rotation:NiagaraFloat=&RandomRangeFloat(Minimum=0, Maximum=360)
    Particles.Flash Star Sprite Size:Vector2f=&Multiply_Vector2D_ByFloat(
     Float=&Multiply_Float(
      A=&MakeCustomFloatFromBool(Bool=$User.bIsFrontEnd, True Float=$User.FrontEndScale, False Float=1)
      B=&RandomRangeFloat(Minimum=40, Maximum=45)
     )
     Vector2D=$Particles.SpriteSize
    )
    Particles.Flash Star SubUV Index:NiagaraFloat=&RandomRangeFloat(Minimum=0, Maximum=3)
   ) {isSetParameters=true}
   /SetVariables(
    Particles.Flash Flare Color:LinearColor=<rgba 5 5 5 0.01>
    Particles.Flash Flare Lifetime=0.15
    Particles.Flash Flare Sprite Size:Vector2f=&Multiply_Vector2D_ByFloat(Float=&FN_FrontEndFloatModifier(Gameplay=1, Front End=$User.FrontEndScale), Vector2D=<90 90>)
   ) {isSetParameters=true}
   &DynamicMaterialParameters(
    Index 0 Param 1=0.2
    Index 0 Param 2=1
    Module.Index0Param0Write=true
    Module.Index0Param1Write=true
    Module.Index0Param2Write=true
    Module.Index0Param3Write=true
   )
   &CameraOffset(Camera Offset Amount=15)
  ]
  particleUpdateScript [
   &ParticleState()
   /SetVariables(
    Particles.Variant0_CC:LinearColor=&MakeCustomLinearColorFromBool(Bool=$User.bIsFrontEnd, True Color=<rgba 2 1.4931 0.3215 0.5>, False Color=<rgba 2 1.0083 0.3061 0.5>)
    Particles.Variant1_CC:LinearColor=&MakeCustomLinearColorFromBool(Bool=$User.bIsFrontEnd, True Color=<rgba 2 1.4931 0.3215 0.5>, False Color=<rgba 2 1.0083 0.3061 0.5>)
   ) {isSetParameters=true}
   /SetVariables(
    Particles.Flash Flare Age:NiagaraFloat=&Divide_Float(A=$Particles.Age, B=$Particles.Initial.Flash Flare Lifetime)
    Particles.Flash Spike Age:NiagaraFloat=&Divide_Float(A=$Particles.Age, B=$Particles.Initial.Flash Spike Lifetime)
    Particles.Flash Star Age:NiagaraFloat=&Divide_Float(A=$Particles.Age, B=$Particles.Initial.Flash Star Lifetime)
   ) {isSetParameters=true}
   /SetVariables(
    Particles.Flash Flare Color:LinearColor=&MakeLinearColorFromVectorAndFloat(
     "Vector (RGB)"=&Multiply_Vector(
      A=&Multiply_Vector(
       A=&MakeVectorFromLinearColorRGB(Color=$Particles.Initial.Flash Flare Color)
       B=&MakeVectorFromLinearColorRGB(Color=&LerpLinearColors(StartColor=$Particles.Variant0_CC, EndColor=$Particles.Variant1_CC, LerpFactor=$User.Variant))
      )
      B=&VectorFromFloat(Value=&FN_FrontEndFloatModifier(Gameplay=1.5, Front End=1))
     )
     "Float (Alpha)"=&Multiply_Float(
      A=&MakeFloatFromLinearColor(LinearColor=<rgba 1 1 1 1>, Channel:ENiagaraLinearColor_Channels=::NewEnumerator3)
      B=&FloatFromCurve(
       FloatCurve=DataInterface() {propertyValues {Curve {keys [{arriveTangent=-1.9727, interpMode=RCIM_Cubic, time=0.1, value=1}, <at 1 0 RCIM_Cubic>]}}}
       CurveIndex=$Particles.Flash Flare Age
      )
     )
    )
    Particles.Flash Spike Color:LinearColor=&MakeLinearColorFromVectorAndFloat(
     "Vector (RGB)"=&MakeVectorFromLinearColorRGB(Color=$Particles.Initial.Flash Spike Color)
     "Float (Alpha)"=&FloatFromCurve(FloatCurve=DataInterface() {propertyValues {Curve {keys [<at 0.7 1>,<at 1 0>]}}}, CurveIndex=$Particles.Flash Spike Age)
    )
    Particles.Flash Spike Sprite Size:Vector2f=&Multiply_Vector2D_ByFloat(
     Float=&FloatFromCurve(FloatCurve=DataInterface() {propertyValues {Curve {keys [<at 0 0.6>,<at 1 2>]}}}, CurveIndex=$Particles.Flash Spike Age)
     Vector2D=$Particles.Initial.Flash Spike Sprite Size
    )
    Particles.Flash Star Color:LinearColor=&MultiplyLinearColor(
     Linear Color=$Particles.Initial.Flash Star Color
     Scale Factor=&MakeVector4_FromLinearColor(LinearColor=&LerpLinearColors(StartColor=$Particles.Variant0_CC, EndColor=$Particles.Variant1_CC, LerpFactor=$User.Variant))
    )
    Particles.Flash Star Sprite Size:Vector2f=&Multiply_Vector2D_ByFloat(
     Float=&FloatFromCurve(
      FloatCurve=DataInterface() {propertyValues {Curve {keys [
          {arriveTangent=2.7013, interpMode=RCIM_Cubic, leaveTangent=2.7013, tangentMode=RCTM_User, time=0, value=0}
          {interpMode=RCIM_Cubic, tangentMode=RCTM_User, time=1, value=1}
         ]}}}
      CurveIndex=$Particles.Flash Star Age
     )
     Vector2D=$Particles.Initial.Flash Star Sprite Size
    )
   ) {isSetParameters=true}
   &FN_Light_Attributes(
    Color=&MultiplyLinearColor(
     Linear Color=$System.ColorA
     Scale Factor=&MakeVector4_FromLinearColor(LinearColor=&LerpLinearColors(StartColor=<rgba 1 1 1 1>, EndColor=$System.TOD Color, LerpFactor=0.5))
    )
    Create Color Attribute=true
    Create Exponent Attribute=true
    Create Position Attribute=true
    Create Radius Attribute=true
    Exponent=2
    Radius=&FN_FrontEndFloatModifier(Gameplay=90, Front End=90)
    Scale Color=&FloatFromCurve(
     FloatCurve=DataInterface() {propertyValues {Curve {keys [
         {arriveTangent=-0.5504, interpMode=RCIM_Cubic, leaveTangent=-0.5504, tangentMode=RCTM_User, time=0, value=1}
         {arriveTangent=-0.5426, interpMode=RCIM_Cubic, leaveTangent=-0.5426, tangentMode=RCTM_User, time=0.55, value=0}
        ]}}}
     Scale Curve=7
    )
   )
   &InheritSourceMovement(
    Applied Position Falloff Distance=1000
    Applied Position Scale=&Multiply_VectorByFloat(
     Vector=<1 1 0>
     Float=&FloatFromCurve(
      FloatCurve=DataInterface() {propertyValues {Curve {keys [
          {interpMode=RCIM_Cubic, tangentMode=RCTM_User, time=0.1865, value=1}
          {arriveTangent=-2.1905, interpMode=RCIM_Cubic, leaveTangent=-2.1905, tangentMode=RCTM_User, time=1, value=0}
         ]}}}
      CurveIndex=$Emitter.Age
     )
    )
    ApplyForce=false
    ApplyPosition=true
    ApplyPositionFalloff=false
   )
   &SolveForcesAndVelocity(Clamp Velocity=false, Limit Acceleration=false)
  ]
  rendererClasses [&NiagaraSpriteRendererProperties]
  renderers [
   {rendererClass=&NiagaraSpriteRendererProperties}
   {rendererClass=&NiagaraSpriteRendererProperties, rendererIndex=1}
   {rendererClass=&NiagaraSpriteRendererProperties, rendererIndex=2}
  ]
 }
 FlowerParent {
  emitterUpdateScript [
   &EmitterState()
   &SpawnBurst_Instantaneous(Spawn Count=90, Use Spawn Probability=false, UseLoopCountLimit=false)
   &SpawnBurst_Instantaneous(Spawn Count=19, Spawn Group=1, Use Spawn Probability=false, UseLoopCountLimit=false) {module=SpawnBurst_Instantaneous001}
  ]
  particleSpawnScript [
   /SetVariables(Particles.SpawnGroup=$Emitter.SpawnGroup) {isSetParameters=true}
   /SetVariables(Particles.VisibilityTag=$Particles.SpawnGroup) {isSetParameters=true}
   &InitializeParticle(
    Color=<rgba 0.849 0.8394 0.6928 1>
    Lifetime Max=3.5
    Lifetime Min=2.75
    Lifetime Mode=::NewEnumerator1
    Mesh Scale Mode=::NewEnumerator3
    Mesh Uniform Scale=&Lerp_Float(
     A=&RandomRangeFloat(Minimum=0.045, Maximum=0.036)
     B=&RandomRangeFloat(Minimum=0.024, Maximum=0.042)
     Alpha=&MakeCustomFloatFromBool(
      Bool=&SetBoolByIntComparison(A=1, B=$Particles.SpawnGroup, Comparison Type=::NewEnumerator4)
      True Float=1
      False Float=0
     )
    )
    Sprite Rotation Mode=::NewEnumerator2
    Sprite Size Mode=::NewEnumerator3
    Uniform Sprite Size=&Lerp_Float(
     A=&RandomRangeFloat(Minimum=4, Maximum=5)
     B=&RandomRangeFloat(Minimum=3, Maximum=4)
     Alpha=&MakeCustomFloatFromBool(
      Bool=&SetBoolByIntComparison(A=1, B=$Particles.SpawnGroup, Comparison Type=::NewEnumerator4)
      True Float=1
      False Float=0
     )
    )
    UsePositionOffset=false
    Write Material Random=true
   )
   &ShapeLocation(MaskSpawnGroup=false, Override Seed=false, Sphere Radius=5, Sphere Surface Distribution=1, UseOwnerScale=true)
   &AddVelocity(
    Module.Use Falloff Distance=false
    Origin Offset=<0 0 -4>
    Override Seed=false
    Velocity Mode=::NewEnumerator1
    Velocity Speed=&RandomRangeFloat(Minimum=90, Maximum=140)
   )
   &AddVelocity(
    Cone Angle=110
    Cone Axis=<0 0 1>
    Module.UseConeFalloff=false
    Module.UseVelDistribution=false
    Override Seed=false
    Rotation Coordinate Space=::Simulation
    Velocity Mode=::NewEnumerator2
    Velocity Speed=&RandomRangeFloat(Minimum=10, Maximum=30)
   ) {enabled=false, module=AddVelocity001}
   &CurlNoiseForce(Bias Noise=false, Pan Noise=false) {enabled=false, module=CurlNoiseForce001}
   &AddVelocity(
    Module.Use Falloff Distance=false
    Origin Offset=<0 0 -1>
    Override Seed=false
    Velocity Mode=::NewEnumerator1
    Velocity Speed=/Niagara/DynamicInputs/Math/LerpMultipleFloats(
     Float Count:ENiagara_MultipleLerpCount=::NewEnumerator0
     1:NiagaraFloat=&RandomRangeFloat(Minimum=50, Maximum=112.5)
     2:NiagaraFloat=&RandomRangeFloat(Minimum=62.5, Maximum=162.5)
     3:NiagaraFloat=&RandomRangeFloat(Minimum=50, Maximum=150)
     4:NiagaraFloat=?
     5:NiagaraFloat=?
     Alpha:NiagaraFloat=&RandomRangeFloat(Minimum=0, Maximum=1)
    )
   ) {module=AddVelocity002}
   &CurlNoiseForce(Bias Noise=false, Noise Frequency=10, Noise Strength=100, Pan Noise=false, Pan Noise Field=<0.2016 0 0.5008>) {module=CurlNoiseForce002}
   &ApplyInitialForces()
   &SubUVAnimation(
    Direct Frame Index:NiagaraFloat=&Lerp_Float(
     A=&SelectFloatFromArray(
      Array Sampling Mode=::NewEnumerator0
      Direct Array Index=?
      Float Selection Array=DataInterface() {propertyValues {FloatData [0,1,4]}}
      Random Seed=0
     )
     B=1
     Alpha=&MakeCustomFloatFromBool(
      Bool=&SetBoolByIntComparison(A=1, B=$Particles.SpawnGroup, Comparison Type=::NewEnumerator4)
      True Float=1
      False Float=0
     )
    )
    "Sprite Renderer "=DataInterface() {propertyValues {SpriteRenderer=&self:FlowerParent_0.NiagaraSpriteRendererProperties_0}}
    SubUV Animation Mode:ENiagara_SubUVLookupModeV2=::NewEnumerator5
   )
   &SubUVAnimation(
    Direct Frame Index:NiagaraFloat=&Lerp_Float(
     A=&SelectFloatFromArray(
      Array Sampling Mode=::NewEnumerator0
      Direct Array Index=?
      Float Selection Array=DataInterface() {propertyValues {FloatData [0,1,4]}}
      Random Seed=0
     )
     B=3
     Alpha=&MakeCustomFloatFromBool(
      Bool=&SetBoolByIntComparison(A=1, B=$Particles.SpawnGroup, Comparison Type=::NewEnumerator4)
      True Float=1
      False Float=0
     )
    )
    "Sprite Renderer "=DataInterface() {propertyValues {SpriteRenderer=&self:FlowerParent_0.NiagaraSpriteRendererProperties_1}}
    SubUV Animation Mode:ENiagara_SubUVLookupModeV2=::NewEnumerator5
   ) {module=SubUVAnimation001}
   &SubUVAnimation(
    Direct Frame Index:NiagaraFloat=&Lerp_Float(
     A=&SelectFloatFromArray(
      Array Sampling Mode=::NewEnumerator0
      Direct Array Index=?
      Float Selection Array=DataInterface() {propertyValues {FloatData [0,1,4]}}
      Random Seed=0
     )
     B=1
     Alpha=&MakeCustomFloatFromBool(
      Bool=&SetBoolByIntComparison(A=1, B=$Particles.SpawnGroup, Comparison Type=::NewEnumerator4)
      True Float=1
      False Float=0
     )
    )
    Mesh Renderer=DataInterface() {propertyValues {MeshRenderer=&self:FlowerParent_0.NiagaraMeshRendererProperties_0}}
    Renderer Type:ENiagaraMeshOrSprite=::NewEnumerator1
    SubUV Animation Mode:ENiagara_SubUVLookupModeV2=::NewEnumerator5
   ) {module=SubUVAnimation003}
   &SubUVAnimation(
    Direct Frame Index:NiagaraFloat=&Lerp_Float(
     A=&SelectFloatFromArray(
      Array Sampling Mode=::NewEnumerator0
      Direct Array Index=?
      Float Selection Array=DataInterface() {propertyValues {FloatData [0,1,4]}}
      Random Seed=0
     )
     B=3
     Alpha=&MakeCustomFloatFromBool(
      Bool=&SetBoolByIntComparison(A=1, B=$Particles.SpawnGroup, Comparison Type=::NewEnumerator4)
      True Float=1
      False Float=0
     )
    )
    Mesh Renderer=DataInterface() {propertyValues {MeshRenderer=&self:FlowerParent_0.NiagaraMeshRendererProperties_1}}
    Renderer Type:ENiagaraMeshOrSprite=::NewEnumerator1
    SubUV Animation Mode:ENiagara_SubUVLookupModeV2=::NewEnumerator5
   ) {module=SubUVAnimation002}
   &InitialMeshOrientation(
    Mesh Orientation Mode=::NewEnumerator2
    Rotation=&RandomRangeVector(Minimum=<0.3 -0.2 -0.2>, Maximum=<0.7 0.2 0.2>)
    Use Rotation Vector=true
   )
  ]
  particleUpdateScript [
   &ParticleState()
   &SpriteRotationRate(Rotation Rate=&SelectFloatFromArray(
     Array Sampling Mode=::NewEnumerator0
     Direct Array Index=?
     Float Selection Array=DataInterface() {propertyValues {FloatData [225,200,175,-175,-200,-225]}}
     Random Seed=0
    )) {module=SpriteRotationRate001}
   &MeshRotationForce(
    Coordinate Space:ENiagaraMeshTransforms=::NewEnumerator3
    "Lever Radius (cm)"=70
    Rotation=&RandomRangeVector(Minimum=<3 3 7>, Maximum=<5 5 9>)
   )
   &PointForce(Force Strength=266.9367, Module.Use Falloff=true, Module.UseLinearForce=false)
   &GravityForce()
   &CurlNoiseForce(Noise Frequency=25, Noise Strength=&RandomRangeFloat(Minimum=200, Maximum=250), Pan Noise=false)
   &Drag(
    Drag=&FloatFromCurve(
     FloatCurve=DataInterface() {propertyValues {Curve {keys [
         {arriveTangent=2.2311, interpMode=RCIM_Cubic, leaveTangent=2.2311, tangentMode=RCTM_User, time=0, value=0.2}
         {arriveTangent=3.3414, interpMode=RCIM_Cubic, leaveTangent=3.3414, tangentMode=RCTM_User, time=0.25, value=1}
        ]}}}
     Scale Curve=&MakeCustomFloatFromBool(
      Bool=&SetBoolByIntComparison(A=1, B=$Particles.SpawnGroup, Comparison Type=::NewEnumerator4)
      True Float=&RandomRangeFloat(Minimum=1.5, Maximum=2)
      False Float=&RandomRangeFloat(Minimum=1.5, Maximum=2)
     )
    )
    Use Linear Drag=true
    Use Rotational Drag=false
   ) {enabled=false}
   &AerodynamicDrag(Use Linear Drag=true, Use Rotational Drag=true)
   &Collision(
    Age Colliding Particles=false
    CPU Trace Vector Length Multiplier=0.8
    Enable Rest State=false
    Friction=0.5
    Module.EnableMaxCollisionCount=false
    Module.Randomize Collision Normal Vector=false
    Particle Radius Scale=0
    Restitution=&RandomRangeFloat(Minimum=0.3, Maximum=0.7)
    Simple Friction=false
   ) {enabled=false}
   &Collision(
    Age Colliding Particles=false
    CPU Trace Vector Length Multiplier=0.8
    Collision Enabled=&Boolean_OrOperation(
     Bool1=&IsPlatformSetActive(New Platform Set=DataInterface() {propertyValues {Platforms {qualityLevelMask=-4}}})
     Bool2=$User.bIsFrontEndPreview
    )
    Enable Rest State=false
    Friction=0.5
    Module.EnableMaxCollisionCount=false
    Module.Randomize Collision Normal Vector=false
    Particle Radius Scale=0
    Radius Calculation Type=::NewEnumerator1
    Restitution=&RandomRangeFloat(Minimum=0.3, Maximum=0.7)
    Simple Friction=false
   ) {module=Collision001}
   &ScaleSpriteSize(Scale Factor=&Vector2DFromFloat(Value=&FloatFromCurve(FloatCurve=DataInterface() {propertyValues {Curve {keys [
          <at -0.0046 1.0113 RCIM_Cubic>
          {arriveTangent=-1.0036, interpMode=RCIM_Cubic, leaveTangent=-1.0036, time=0.7001, value=0.872}
          <at 1.0011 0.002>
         ]}}})))
   &ScaleMeshSize(Scale Factor=&Multiply_VectorByFloat(
     Vector=&RandomRangeVector(Minimum=<0.9 0.9 0.4>, Maximum=<1 1 0.5>)
     Float=&FloatFromCurve(FloatCurve=DataInterface() {propertyValues {Curve {keys [
          <at 0 0 RCIM_Cubic>
          {arriveTangent=1.2455, interpMode=RCIM_Cubic, leaveTangent=1.2455, time=0.05, value=1}
          {arriveTangent=-1.0493, interpMode=RCIM_Cubic, leaveTangent=-1.0493, time=0.7001, value=0.872}
          <at 1.0011 0.002>
         ]}}})
    ))
   &ScaleColor(
    Scale Alpha=&FloatFromCurve(FloatCurve=DataInterface() {propertyValues {Curve {keys [<at -0.0046 1.0113 RCIM_Cubic>,<at 1.0011 0.002 RCIM_Cubic>]}}})
    ScaleA=true
    ScaleRGB=true
   )
   &InheritSourceMovement(
    Applied Position Falloff Distance=1000
    Applied Position Scale=&Multiply_VectorByFloat(
     Vector=<1 1 0>
     Float=&FloatFromCurve(
      FloatCurve=DataInterface() {propertyValues {Curve {keys [
          {interpMode=RCIM_Cubic, tangentMode=RCTM_User, time=0.1865, value=1}
          {arriveTangent=-2.1905, interpMode=RCIM_Cubic, leaveTangent=-2.1905, tangentMode=RCTM_User, time=1, value=0}
         ]}}}
      CurveIndex=$Emitter.Age
     )
    )
    ApplyForce=false
    ApplyPosition=true
    ApplyPositionFalloff=false
   )
   &SolveForcesAndVelocity(Clamp Velocity=false, Limit Acceleration=false)
   &AlignSpriteToMeshOrientation(SetAlignment=true, SetFacing=true)
  ]
  rendererClasses [&NiagaraSpriteRendererProperties, &NiagaraMeshRendererProperties]
  renderers [
   {rendererClass=&NiagaraSpriteRendererProperties}
   {rendererClass=&NiagaraSpriteRendererProperties, rendererIndex=1}
   {rendererClass=&NiagaraMeshRendererProperties, rendererIndex=2}
   {rendererClass=&NiagaraMeshRendererProperties, rendererIndex=3}
  ]
 }
 Sun_Light {
  emitterUpdateScript [
   &EmitterState()
   &SpawnBurst_Instantaneous(Use Spawn Probability=false, UseLoopCountLimit=false)
  ]
  particleSpawnScript [
   &InitializeParticle(
    Color=&MakeCustomLinearColorFromBool(Bool=$User.bIsFrontEnd, True Color=<rgba 2 1.4931 0.3215 0.5>, False Color=<rgba 2 1.0083 0.3061 0.5>)
    Lifetime=0.3
    Mesh Scale Mode=::NewEnumerator3
    Sprite Size Mode=::NewEnumerator3
    Uniform Sprite Size=30
    UsePositionOffset=true
    Write Material Random=true
   )
   &SystemLocation(MaskSpawnGroup=false)
  ]
  particleUpdateScript [
   &ParticleState()
   &ScaleSpriteSize(Uniform Curve Sprite Scale=DataInterface() {propertyValues {Curve {keys [
        {arriveTangent=4.1785, interpMode=RCIM_Cubic, leaveTangent=4.1785, tangentMode=RCTM_User, time=0, value=0}
        {arriveTangent=1.1179, interpMode=RCIM_Cubic, leaveTangent=1.1179, tangentMode=RCTM_User, time=0.4, value=1.1}
        {arriveTangent=0.0224, interpMode=RCIM_Cubic, leaveTangent=0.0224, tangentMode=RCTM_User, time=1, value=1.5}
       ]}}})
   /SetVariables(Particles.LightColor=<rgba 1 0.807 0.3607 9>) {isSetParameters=true}
   &ScaleColor(
    Scale Alpha=&FloatFromCurve(FloatCurve=DataInterface() {propertyValues {Curve {keys [
         {tangentMode=RCTM_User, time=0, value=1}
         {arriveTangent=-0.0405, interpMode=RCIM_Cubic, leaveTangent=-0.0405, tangentMode=RCTM_User, time=0.4, value=1}
         {arriveTangent=-2.6152, interpMode=RCIM_Cubic, leaveTangent=-2.6152, tangentMode=RCTM_User, time=1, value=0}
        ]}}})
    ScaleA=true
    ScaleRGB=true
   )
   &InheritSourceMovement(
    Applied Position Falloff Distance=1000
    Applied Position Scale=&Multiply_VectorByFloat(
     Vector=<1 1 0>
     Float=&FloatFromCurve(
      FloatCurve=DataInterface() {propertyValues {Curve {keys [
          {interpMode=RCIM_Cubic, tangentMode=RCTM_User, time=0.1865, value=1}
          {arriveTangent=-2.1905, interpMode=RCIM_Cubic, leaveTangent=-2.1905, tangentMode=RCTM_User, time=1, value=0}
         ]}}}
      CurveIndex=$Emitter.Age
     )
    )
    ApplyForce=false
    ApplyPosition=true
    ApplyPositionFalloff=false
   )
   &SolveForcesAndVelocity(Clamp Velocity=false, Limit Acceleration=false)
  ]
  rendererClasses [
   &NiagaraSpriteRendererProperties
   &NiagaraLightRendererProperties
   &NiagaraComponentRendererProperties
  ]
  renderers [
   {rendererClass=&NiagaraSpriteRendererProperties}
   {rendererClass=&NiagaraLightRendererProperties, rendererIndex=1}
   {rendererClass=&NiagaraComponentRendererProperties, rendererIndex=2}
  ]
 }
 Sun_Rays {
  emitterUpdateScript [
   &EmitterState()
   &SpawnBurst_Instantaneous(Use Spawn Probability=false, UseLoopCountLimit=false)
  ]
  particleSpawnScript [
   &InitializeParticle(
    Color=&MakeCustomLinearColorFromBool(Bool=$User.bIsFrontEnd, True Color=<rgba 11 7.7647 0.2588 0.01>, False Color=<rgba 15 9.163 1.3596 0.01>)
    Lifetime=0.3
    Sprite Size Mode=::NewEnumerator3
    Uniform Sprite Size=101.1211
    UsePositionOffset=true
    Write Material Random=true
   )
   &SystemLocation(MaskSpawnGroup=false)
  ]
  particleUpdateScript [
   &ParticleState()
   &ScaleSpriteSize(Uniform Curve Sprite Scale=DataInterface() {propertyValues {Curve {keys [
        {arriveTangent=4.1785, interpMode=RCIM_Cubic, leaveTangent=4.1785, tangentMode=RCTM_User, time=0, value=0}
        {arriveTangent=1.1179, interpMode=RCIM_Cubic, leaveTangent=1.1179, tangentMode=RCTM_User, time=0.2042, value=0.6228}
        {arriveTangent=0.0224, interpMode=RCIM_Cubic, leaveTangent=0.0224, tangentMode=RCTM_User, time=1, value=1}
       ]}}})
   &SpriteRotationRate(Rotation Rate=&Sine(Normalized Angle=$Emitter.Age, Period=5, Scale=88, Bias=0)) {module=SpriteRotationRate001}
   &ScaleColor(
    Scale Alpha=&FloatFromCurve(FloatCurve=DataInterface() {propertyValues {Curve {keys [
         {tangentMode=RCTM_User, time=0, value=1}
         {arriveTangent=-0.0405, interpMode=RCIM_Cubic, leaveTangent=-0.0405, tangentMode=RCTM_User, time=0.4, value=1}
         {arriveTangent=-2.6152, interpMode=RCIM_Cubic, leaveTangent=-2.6152, tangentMode=RCTM_User, time=1, value=0}
        ]}}})
    ScaleA=true
    ScaleRGB=true
   )
   &InheritSourceMovement(
    Applied Position Falloff Distance=1000
    Applied Position Scale=&Multiply_VectorByFloat(
     Vector=<1 1 0>
     Float=&FloatFromCurve(
      FloatCurve=DataInterface() {propertyValues {Curve {keys [
          {interpMode=RCIM_Cubic, tangentMode=RCTM_User, time=0.1865, value=1}
          {arriveTangent=-2.1905, interpMode=RCIM_Cubic, leaveTangent=-2.1905, tangentMode=RCTM_User, time=1, value=0}
         ]}}}
      CurveIndex=$Emitter.Age
     )
    )
    ApplyForce=false
    ApplyPosition=true
    ApplyPositionFalloff=false
   )
   &SolveForcesAndVelocity(Clamp Velocity=false, Limit Acceleration=false)
  ]
  rendererClasses [&NiagaraSpriteRendererProperties]
  renderers [{rendererClass=&NiagaraSpriteRendererProperties}]
 }
 Sun_Rays001 {
  emitterUpdateScript [
   &EmitterState()
   &SpawnBurst_Instantaneous(Use Spawn Probability=false, UseLoopCountLimit=false)
  ]
  particleSpawnScript [
   &InitializeParticle(
    Color=&MakeCustomLinearColorFromBool(Bool=$User.bIsFrontEnd, True Color=<rgba 11 7.7647 0.2588 0.01>, False Color=<rgba 15 9.163 1.3596 0.01>)
    Lifetime=0.3
    Sprite Rotation Mode=::NewEnumerator2
    Sprite Size Mode=::NewEnumerator3
    Uniform Sprite Size=108.6472
    UsePositionOffset=true
    Write Material Random=true
   )
   &SystemLocation(MaskSpawnGroup=false)
  ]
  particleUpdateScript [
   &ParticleState()
   &ScaleSpriteSize(Uniform Curve Sprite Scale=DataInterface() {propertyValues {Curve {keys [
        {arriveTangent=4.1785, interpMode=RCIM_Cubic, leaveTangent=4.1785, tangentMode=RCTM_User, time=0, value=0}
        {arriveTangent=1.1179, interpMode=RCIM_Cubic, leaveTangent=1.1179, tangentMode=RCTM_User, time=0.2042, value=0.6228}
        {arriveTangent=0.0224, interpMode=RCIM_Cubic, leaveTangent=0.0224, tangentMode=RCTM_User, time=1, value=1}
       ]}}})
   &SpriteRotationRate(Rotation Rate=&Sine(Normalized Angle=$Emitter.Age, Period=2.4044, Scale=135.79, Bias=0)) {module=SpriteRotationRate001}
   &ScaleColor(
    Scale Alpha=&FloatFromCurve(FloatCurve=DataInterface() {propertyValues {Curve {keys [
         {tangentMode=RCTM_User, time=0, value=1}
         {arriveTangent=-0.0405, interpMode=RCIM_Cubic, leaveTangent=-0.0405, tangentMode=RCTM_User, time=0.4, value=1}
         {arriveTangent=-2.6152, interpMode=RCIM_Cubic, leaveTangent=-2.6152, tangentMode=RCTM_User, time=1, value=0}
        ]}}})
    ScaleA=true
    ScaleRGB=true
   )
   &InheritSourceMovement(
    Applied Position Falloff Distance=1000
    Applied Position Scale=&Multiply_VectorByFloat(
     Vector=<1 1 0>
     Float=&FloatFromCurve(
      FloatCurve=DataInterface() {propertyValues {Curve {keys [
          {interpMode=RCIM_Cubic, tangentMode=RCTM_User, time=0.1865, value=1}
          {arriveTangent=-2.1905, interpMode=RCIM_Cubic, leaveTangent=-2.1905, tangentMode=RCTM_User, time=1, value=0}
         ]}}}
      CurveIndex=$Emitter.Age
     )
    )
    ApplyForce=false
    ApplyPosition=true
    ApplyPositionFalloff=false
   )
   &SolveForcesAndVelocity(Clamp Velocity=false, Limit Acceleration=false)
  ]
  rendererClasses [&NiagaraSpriteRendererProperties]
  renderers [{rendererClass=&NiagaraSpriteRendererProperties}]
 }
}
summary {
 systemName=NS_Emote_SunnySteppin
 emitters [
  {emitterName=Sun_Rays, rendererClasses [&NiagaraSpriteRendererProperties]}
  {
   emitterName=Sun_Light
   rendererClasses [
    &NiagaraSpriteRendererProperties
    &NiagaraLightRendererProperties
    &NiagaraComponentRendererProperties
   ]
  }
  {emitterName=FlowerParent, rendererClasses [&NiagaraSpriteRendererProperties, &NiagaraMeshRendererProperties]}
  {emitterName=Sun_Rays001, rendererClasses [&NiagaraSpriteRendererProperties]}
  {emitterName=FlashElements, rendererClasses [&NiagaraSpriteRendererProperties]}
 ]
 userVariables [
  User.bIsFrontEnd {type=NiagaraBool}
  User.bIsFrontEndPreview {type=NiagaraBool}
  User.FrontEndScale {type=NiagaraFloat, defaultValue=0.5, description=""}
  User.New Skeletal Mesh {
   type=NiagaraDataInterfaceSkeletalMesh
   defaultValue {
    struct=NiagaraExt_VariableValue_DataInterface
    value {dataInterface=&self:NiagaraDataInterfaceSkeletalMesh_1}
   }
  }
  User.SkeletalMesh {
   type=NiagaraDataInterfaceSkeletalMesh
   defaultValue {
    struct=NiagaraExt_VariableValue_DataInterface
    value {dataInterface=&self:NiagaraDataInterfaceSkeletalMesh_0}
   }
  }
  User.Variant {type=NiagaraFloat, defaultValue=0, description=""}
 ]
}
